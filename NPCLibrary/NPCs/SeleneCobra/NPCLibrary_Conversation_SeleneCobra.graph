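# Conversation database for the SeleneCobra NPC, a scout specialist, defining its greeting and the conversation branches that follow.
# The entry point leads to a Speech node whose contextual lines react to first-ever meetings, specific player outfits and per-match conversation count.
# Curve rows set priority for first-meeting and outfit lines.
# The greeting fans out through link nodes to shared quest and basic-service conversations, a Season 29 Snapshot quest, and Selene's own dialogue entry.
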ConversationDatabase /NPCLibrary/NPCs/SeleneCobra/NPCLibrary_Conversation_SeleneCobra
entry_points [{tag=Conversation.Entry.SeleneCobra, destinations [#1]}]
nodes [
 ConversationEntryPointNode_7 {
  id=#1
  class=ConversationEntryPointNode
  outputs [#2]
  properties {entryTag {tagName=Conversation.Entry.SeleneCobra}}
 }
 Speech_C_4 {
  id=#2
  class=Speech_C
  outputs [#3, #4, #5, #6]
  properties {
   nodeName=""
   generalConfig {
    defaultMessage="You're getting quieter, little mouse. I'm impressed."
    contextualMessages [
     {
      message="Well met, little mouse."
      contextRequirements [{requirement {class=FortControllerRequirement_IsFirstEverConversationWithNPC}}]
      priority {curve {rowName=Tandem.Speech.Priority.FirstEverInteraction}}
     }
     {
      message=You and The Oracle have developed a kinship. Seems appropriate given your... lineage.
      contextRequirements [{requirement {
         class=FortControllerRequirement_HasCID
         properties {allowedCIds [
           &Character_VoidRedemption_Rebel
           &Character_BrightDisk
          ]}
        }}]
      priority {curve {rowName=Tandem.Speech.Priority.OutfitReaction}}
     }
     {
      message=You have been a joy to hunt. Let's not end this dance prematurely.
      contextRequirements [{requirement {
         class=FortControllerRequirement_HasCID
         properties {allowedCIds [&Character_BengalBasher_Suit]}
        }}]
      priority {curve {rowName=Tandem.Speech.Priority.OutfitReaction}}
     }
     {
      message="You have changed the world, Medusa. Be ready to shape a better one."
      contextRequirements [{requirement {
         class=FortControllerRequirement_HasCID
         properties {allowedCIds [/BRCosmetics/Athena/Items/Cosmetics/Characters/Character_ReptilianOcean_Sleek]}
        }}]
      priority {curve {rowName=Tandem.Speech.Priority.OutfitReaction}}
     }
     {
      message="You're getting quieter, little mouse. I'm impressed."
      contextRequirements [{requirement {
         class=FortControllerRequirement_CheckNPCMatchConversationCount
         properties {nPCUniqueIdTag {tagName=AISpawnerData.Type.Specialist.Scout.SeleneCobra}}
        }}]
      priority {curve {rowName=None}}
     }
     {
      message="The hunt awaits us. "
      contextRequirements [{requirement {
         class=FortControllerRequirement_CheckNPCMatchConversationCount
         properties {nPCUniqueIdTag {tagName=AISpawnerData.Type.Specialist.Scout.SeleneCobra}}
        }}]
      priority {curve {rowName=None}}
     }
     {
      message=I have not taken mortals as hunting companions in eons. Mayhaps I will change that...
      contextRequirements [{requirement {
         class=FortControllerRequirement_CheckNPCMatchConversationCount
         properties {nPCUniqueIdTag {tagName=AISpawnerData.Type.Specialist.Scout.SeleneCobra}}
        }}]
      priority {curve {rowName=None}}
     }
    ]
   }
  }
 }
 ConversationLinkNode_3 {
  id=#6
  class=ConversationLinkNode
  properties {remoteEntryTag {tagName=Conversation.Entry.SeleneCobra.Dialogue}}
 }
 ConversationLinkNode_2 {
  id=#5
  class=ConversationLinkNode
  properties {remoteEntryTag {tagName=Conversation.Entry.Season29.Snapshot.P02.Q12}}
 }
 ConversationLinkNode_0 {id=#4, class=ConversationLinkNode, properties {remoteEntryTag {tagName=Conversation.Entry.BasicServices}}}
 ConversationLinkNode_1 {id=#3, class=ConversationLinkNode, properties {remoteEntryTag {tagName=Conversation.Entry.NPCLibrary.Quests}}}
]
